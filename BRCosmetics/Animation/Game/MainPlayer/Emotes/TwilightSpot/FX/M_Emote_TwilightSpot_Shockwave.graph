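# Translucent, unlit, two-sided material for Niagara mesh particles, drawing the shockwave effect of the TwilightSpot emote.
# Particle color RGB feeds an emissive adjustment function.
# A tiled shockwave texture is masked by a SmoothStep erosion combining noise, the V coordinate and the Erosion dynamic parameter, then depth-faded and scaled by particle alpha.
# The graph also holds a separate smoke-disk mask branch and a DebugTimeSine call.
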Material /BRCosmetics/Animation/Game/MainPlayer/Emotes/TwilightSpot/FX/M_Emote_TwilightSpot_Shockwave
graph {
 properties {
  bAutomaticallySetUsageInEditor=false
  bUsedWithNiagaraMeshParticles=true
  blendMode=BLEND_Translucent
  previewMesh=/BRCosmetics/Effects/Fort_Effects/Emotes/EmeraldGlass_Green/SM_EmeraldGlass_D_Cone2
  shadingModel=MSM_Unlit
  twoSided=true
  usedShadingModels=Unlit
 }
 nodes [
  n1 = ParticleColor()
  MaterialFunctionCall(Emissive=@n1.RGB) {materialFunction=&MF_FX_EmissiveAdjustments}
  MaterialFunctionCall() {materialFunction=&DebugTimeSine}
  n2 = TextureCoordinate()
  n3 = AppendVector(A=ScalarParameter() {defaultValue=1, parameterName=TextureTilingX}, B=ScalarParameter() {defaultValue=1, parameterName=TextureTilingY})
  n4 = ScalarParameter() {parameterName=TexturePositionX}
  n5 = ScalarParameter() {parameterName=TexturePositionY}
  n6 = TextureSample(UVs=Add(A=Multiply(A=@n2, B=@n3), B=AppendVector(A=@n4, B=@n5)) {constB=0}) {texture=&/T_Emote_TwilightSpot_Shockwave}
  n7 = ComponentMask(None=TextureCoordinate()) {g=true}
  n8 = DynamicParameter() {paramNames [Erosion,Param2,Param3,Param4]}
  n9 = ScalarParameter() {defaultValue=1, parameterName=ErosionTilingX}
  n10 = ScalarParameter() {defaultValue=1, parameterName=ErosionTilingY}
  n11 = TextureSampleParameter2D(UVs=Multiply(A=TextureCoordinate(), B=AppendVector(A=@n9, B=@n10))) {parameterName=ErosionTexture, samplerType=SAMPLERTYPE_Masks, texture=/Game/Accessories/FORT_Backpacks/M_MED_Heist/Textures/T_CrystalNoise_M}
  n12 = ScalarParameter() {defaultValue=-0.05, parameterName=ErosionTextureAdd}
  n13 = MaterialFunctionCall(Alpha=Saturate(None=Add(A=@n7, B=@n8.Erosion)), Min=Saturate(None=Add(A=@n11.R, B=@n12) {constB=-0.05})) {materialFunction=&SmoothStep}
  Multiply(A=@n1.A, B=DepthFade(Opacity=Multiply(A=@n6.R, B=@n13.Result), FadeDistance=ScalarParameter() {parameterName=DepthFade}))
  n14 = ScalarParameter() {defaultValue=1, desc=0 or 1 value, parameterName=TextureSelector}
  n15 = Multiply(
   A=TextureCoordinate()
   B=AppendVector(A=ScalarParameter() {defaultValue=1, parameterName=Tiling}, B=LinearInterpolate(Alpha=@n14) {constA=-0.4, constAlpha=1, constB=0.55})
  )
  n16 = ScalarParameter() {defaultValue=-0.13, parameterName=UVPositioning}
  n17 = TextureCoordinate()
  Multiply(
   A=TextureSampleParameter2D(UVs=Add(A=@n15, B=@n16) {constB=-0.13}) {parameterName=Texture, texture=/Game/Effects/Fort_Effects/Textures/Smoke/T_Smoke_Bomb_Disk2}.R
   B=MaterialFunctionCall(In=ComponentMask(None=@n17) {g=true}, Contrast=Constant() {r=1}) {materialFunction=&CheapContrast}.Result
  )
 ]
}
