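# Niagara impact effect for the TwilightSpot emote: a brief, blue-tinted burst that mostly pushes along +X.
# Burst emitters layer stretching sprite streaks, drag-slowed sparks, a ring of X streaks, distance-culled spike streaks, a scaling shockwave and a toon mesh flare.
# Time-of-day inputs set color or alpha per layer, while curves drive size, fades and dynamic material parameters.
# The secondary NE_Impact001 streak emitter is disabled.
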NiagaraSystem /BRCosmetics/Animation/Game/MainPlayer/Emotes/TwilightSpot/FX/NS_Emote_TwilightSpot_Impact
dependencies {
 usedDataInterfaces [
  &NiagaraDataInterfaceCurve
  &NiagaraDataInterfaceArrayFloat3
  &NiagaraDataInterfaceVector2DCurve
  &NiagaraDataInterfaceColorCurve
  &NiagaraDataInterfaceVectorCurve
 ]
 usedDynamicInputs [
  &MaskLinearColorBySpawnGroup
  &SimulationPosition
  &Multiply_Vector
  &Lerp_Vector
 ]
 usedModules [&SystemState]
}
emitters {
 NE_Impact {
  emitterUpdateScript [
   &EmitterState(Life Cycle Mode=::NewEnumerator1, Loop Behavior=::NewEnumerator1, Loop Duration=0.4, UseLoopDelay=false)
   &SpawnBurst_Instantaneous(Spawn Count=3, Use Spawn Probability=false, UseLoopCountLimit=false)
  ]
  particleSpawnScript [
   &InitializeParticle(
    Color Minimum=<rgba 0.2772 0.2772 0.2772 1>
    Color Mode=::NewEnumerator2
    Lifetime Max=0.2
    Lifetime Min=0.15
    Lifetime Mode=::NewEnumerator1
    Sprite Size Max=<7 45>
    Sprite Size Min=<5 35>
    Sprite Size Mode=::NewEnumerator2
    Sprite UV Mode=::NewEnumerator1
    UsePositionOffset=false
    Write Material Random=true
   ) {module=InitializeParticle001}
   &SubUVAnimation(End Frame=3)
   &ShapeLocation(
    MaskSpawnGroup=false
    Non Uniform Scale=<1 1 0.5>
    Offset=&MakeVector(X=1.9352, Y=0, Z=0)
    Offset Mode=::NewEnumerator0
    Sphere Distribution=::NewEnumerator1
    Sphere Radius=2.1403
    UseOwnerScale=true
   )
   &AddVelocity(
    Velocity=&Add_Vector(
     A=&SelectVectorFromArray(
      Direct Array Index=&ReturnExecIndex()
      Vector Selection Array=DataInterface() {propertyValues {FloatData [<3 0 0>, <3 0 0>, <3 0 0>]}}
     )
     B=&RandomRangeVector(Minimum=<0.05 0.05 0.05>, Maximum=<0 -0.05 -0.05>)
    )
    Velocity Speed Scale=15
   ) {module=AddVelocity003}
   &CameraOffset(Camera Offset Amount=10)
  ]
  particleUpdateScript [
   &ParticleState()
   &Drag(Drag=5, Use Linear Drag=true, Use Rotational Drag=true)
   &SolveForcesAndVelocity(Clamp Velocity=false, Limit Acceleration=false)
   &ScaleSpriteSize(Scale Factor=&Vector2DFromCurve(
     Vector2Curve=DataInterface() {propertyValues {XCurve {keys [<at 0 1 RCIM_Cubic>,<at 1 0.2473 RCIM_Cubic>]}, YCurve {keys [<at 0 1.028 RCIM_Cubic>,<at 1 2.8057 RCIM_Cubic>]}}}
     ScaleCurve=&Multiply_Vector2D_ByFloat(Float=1, Vector2D=<1 2>)
    ))
   &ScaleColor(
    Scale Alpha=&FloatFromCurve(
     FloatCurve=DataInterface() {propertyValues {Curve {keys [<at 0 1 RCIM_Cubic>, {arriveTangent=-1.851, tangentMode=RCTM_User, time=1, value=0}]}}}
     Scale Curve=&FNTimeOfDayScalar(Night Scale=0.2, Morning Scale=0.3, Day Scale=0.4, Evening Scale=0.3)
    )
    Scale RGB=&MakeVectorFromLinearColorRGB(Color=&FNTimeOfDayColor(
      Night Color=&ColorFromCurve(ColorCurve=DataInterface() {propertyValues {
         AlphaCurve {keys []}
         BlueCurve {keys [<at 0 0.7263 RCIM_Cubic>,<at 0.7087 0.7772 RCIM_Cubic>]}
         GreenCurve {keys [<at 0 0.2302 RCIM_Cubic>,<at 0.7087 0.2143 RCIM_Cubic>]}
         RedCurve {keys [<at 0 0.1101 RCIM_Cubic>,<at 0.7087 0.0378 RCIM_Cubic>]}
        }})
      Morning Color=&ColorFromCurve(ColorCurve=DataInterface() {propertyValues {
         AlphaCurve {keys []}
         BlueCurve {keys [<at 0 0.7263 RCIM_Cubic>,<at 0.7087 0.7772 RCIM_Cubic>]}
         GreenCurve {keys [<at 0 0.4537 RCIM_Cubic>,<at 0.7087 0.4736 RCIM_Cubic>]}
         RedCurve {keys [<at 0 0.2728 RCIM_Cubic>,<at 0.7087 0.1658 RCIM_Cubic>]}
        }})
      Day Color=&ColorFromCurve(ColorCurve=DataInterface() {propertyValues {
         AlphaCurve {keys []}
         BlueCurve {keys [<at 0 0.7263 RCIM_Cubic>,<at 0.7087 0.7772 RCIM_Cubic>]}
         GreenCurve {keys [<at 0 0.4537 RCIM_Cubic>,<at 0.7087 0.4736 RCIM_Cubic>]}
         RedCurve {keys [<at 0 0.2728 RCIM_Cubic>,<at 0.7087 0.1658 RCIM_Cubic>]}
        }})
      Evening Color=&ColorFromCurve(ColorCurve=DataInterface() {propertyValues {
         AlphaCurve {keys []}
         BlueCurve {keys [<at 0 0.7263 RCIM_Cubic>,<at 0.7087 0.7772 RCIM_Cubic>]}
         GreenCurve {keys [<at 0 0.4537 RCIM_Cubic>,<at 0.7087 0.4736 RCIM_Cubic>]}
         RedCurve {keys [<at 0 0.2728 RCIM_Cubic>,<at 0.7087 0.1658 RCIM_Cubic>]}
        }})
     ))
    ScaleA=true
    ScaleRGB=true
   )
  ]
  rendererClasses [&NiagaraSpriteRendererProperties]
  renderers [{rendererClass=&NiagaraSpriteRendererProperties}]
 }
 NE_Impact001 {
  bEnabled=false
  emitterUpdateScript [
   &EmitterState(Life Cycle Mode=::NewEnumerator1, Loop Behavior=::NewEnumerator1, Loop Duration=0.4, UseLoopDelay=false)
   &SpawnBurst_Instantaneous(Spawn Count=4, Use Spawn Probability=true, UseLoopCountLimit=false)
  ]
  particleSpawnScript [
   &InitializeParticle(
    Color Minimum=<rgba 0.0503 0.0503 0.0503 1>
    Color Mode=::NewEnumerator2
    Lifetime Max=0.2
    Lifetime Min=0.15
    Lifetime Mode=::NewEnumerator1
    Position Offset=<9.9896 0 0>
    Sprite Size Max=<4 7>
    Sprite Size Min=<5 10>
    Sprite Size Mode=::NewEnumerator2
    Sprite UV Mode=::NewEnumerator1
    UsePositionOffset=true
    Write Material Random=true
   ) {module=InitializeParticle001}
   &SubUVAnimation(End Frame=3, SubUV Animation Mode:ENiagara_SubUVLookupMode=::NewEnumerator2)
   &ShapeLocation(
    MaskSpawnGroup=false
    Non Uniform Scale=<1 1 0.5>
    Offset=&MakeVector(X=0, Y=0, Z=0)
    Offset Mode=::NewEnumerator0
    Sphere Distribution=::NewEnumerator1
    Sphere Radius=1
    U Position=&ReturnNormalizedExecIndex()
    UseOwnerScale=true
   )
   &AddVelocity(
    Velocity=&Add_Vector(
     A=&SelectVectorFromArray(
      Direct Array Index=&ReturnExecIndex()
      Vector Selection Array=DataInterface() {propertyValues {FloatData [<-3 0 0>, <-3 0.5686 0.8809>, <-3 -0.5445 -0.7928>]}}
     )
     B=&RandomRangeVector(Minimum=<0.05 0.05 0.05>, Maximum=<-0.05 -0.05 -0.05>)
    )
    Velocity Speed Scale=15
   ) {module=AddVelocity003}
   &CameraOffset(Camera Offset Amount=10)
  ]
  particleUpdateScript [
   &ParticleState()
   &Drag(Drag=5, Use Linear Drag=true, Use Rotational Drag=true)
   &SolveForcesAndVelocity(Clamp Velocity=false, Limit Acceleration=false)
   &ScaleSpriteSize(Scale Factor=&Vector2DFromCurve(
     Vector2Curve=DataInterface() {propertyValues {XCurve {keys [<at 0 1 RCIM_Cubic>,<at 1 0.2473 RCIM_Cubic>]}, YCurve {keys [<at 0 1.9551 RCIM_Cubic>,<at 1 1 RCIM_Cubic>]}}}
     ScaleCurve=&Multiply_Vector2D_ByFloat(Float=1, Vector2D=<1 2>)
    )) {enabled=false}
   &ScaleSpriteSize(Scale Factor=&Vector2DFromCurve(
     Vector2Curve=DataInterface() {propertyValues {XCurve {keys [<at 0 1 RCIM_Cubic>,<at 1 0.2473 RCIM_Cubic>]}, YCurve {keys [<at 0 1.028 RCIM_Cubic>,<at 1 1.3111 RCIM_Cubic>]}}}
     ScaleCurve=&Multiply_Vector2D_ByFloat(Float=1, Vector2D=<1 2>)
    )) {module=ScaleSpriteSize001}
   &ScaleColor(
    Scale Alpha=&FloatFromCurve(
     FloatCurve=DataInterface() {propertyValues {Curve {keys [<at 0 1 RCIM_Cubic>, {arriveTangent=-1.851, tangentMode=RCTM_User, time=1, value=0}]}}}
     Scale Curve=&FNTimeOfDayScalar(Night Scale=0.3, Morning Scale=0.7, Evening Scale=0.7)
    )
    Scale RGB=&MakeVectorFromLinearColorRGB(Color=&ColorFromCurve(ColorCurve=DataInterface() {propertyValues {
        AlphaCurve {keys []}
        BlueCurve {keys [<at 0 1 RCIM_Cubic>,<at 0.7087 0.7772 RCIM_Cubic>]}
        GreenCurve {keys [<at 0 0.7632 RCIM_Cubic>,<at 0.7087 0.4282 RCIM_Cubic>]}
        RedCurve {keys [<at 0 0.5845 RCIM_Cubic>,<at 0.7087 0.2595 RCIM_Cubic>]}
       }}))
    ScaleA=true
    ScaleRGB=true
   )
  ]
  rendererClasses [&NiagaraSpriteRendererProperties]
  renderers [{rendererClass=&NiagaraSpriteRendererProperties}]
 }
 NE_Sparks {
  emitterUpdateScript [
   &EmitterState(Life Cycle Mode=::NewEnumerator1, Loop Behavior=::NewEnumerator1, Loop Duration=1, UseLoopDelay=false)
   &SpawnBurst_Instantaneous(Spawn Count=&RandomRangeInt(Minimum=7, Maximum=5), Use Spawn Probability=false)
  ]
  particleSpawnScript [
   /SetVariables(Particles.Random=&RandomRangeFloat(Minimum=0, Maximum=1)) {isSetParameters=true}
   &InitializeParticle(
    Color=<rgba 1 1 1 0.457>
    Lifetime=&Lerp_Float(A=0.45, B=0.35, Alpha=$Particles.Random)
    Sprite Size Max=<3 35>
    Sprite Size Min=<5 25>
    Sprite Size Mode=::NewEnumerator2
    UsePositionOffset=false
    Write Material Random=true
   ) {module=InitializeParticle001}
   &CameraOffset(Camera Offset Amount=10)
   &SphereLocation(
    MaskSpawnGroup=false
    Override Seed=false
    Sphere Radius=1
    Surface Only=false
    UseNonUniformScale=false
    UseOffset=false
    UseSphereOrientation=false
   )
   &AddVelocity(
    Scale Added Velocity=&Multiply_VectorByFloat(Vector=<1 1 1>, Float=&Lerp_Float(A=2, B=3, Alpha=$Particles.Random))
    Velocity=&RandomRangeVector(Minimum=<20 50 50>, Maximum=<10 -50 -50>)
   ) {module=AddVelocity001}
   &DynamicMaterialParameters(
    Index 0 Param 2=&FloatFromCurve(
     FloatCurve=DataInterface() {propertyValues {Curve {keys [<at 0 0 RCIM_Cubic>, {arriveTangent=1, interpMode=RCIM_Cubic, leaveTangent=1, time=0.7035, value=0.3698}, <at 1 1>]}}}
     CurveIndex=&RandomRangeFloat(Minimum=0, Maximum=1)
    )
    Module.Param0WriteEnabled=true
    Module.Param1WriteEnabled=true
    Module.Param2WriteEnabled=true
    Module.Param3WriteEnabled=true
   ) {module=DynamicMaterialParameters001}
  ]
  particleUpdateScript [
   &ParticleState()
   &Drag(
    Drag=&FloatFromCurve(
     FloatCurve=DataInterface() {propertyValues {Curve {keys [
         {interpMode=RCIM_Cubic, tangentMode=RCTM_User, time=0.5915, value=0}
         {arriveTangent=2.2277, interpMode=RCIM_Cubic, leaveTangent=2.2277, tangentMode=RCTM_User, time=1, value=1}
        ]}}}
     Scale Curve=5
    )
    Use Linear Drag=true
    Use Rotational Drag=true
   )
   &SolveForcesAndVelocity(Clamp Velocity=false, Limit Acceleration=false)
   &DynamicMaterialParameters(
    Index 0 Param 1=&FloatFromCurve(FloatCurve=DataInterface() {propertyValues {Curve {keys [<at 0 0>,<at 1 1>]}}})
    Module.Param0WriteEnabled=true
    Module.Param1WriteEnabled=false
    Module.Param2WriteEnabled=true
    Module.Param3WriteEnabled=true
   )
   &ScaleColor(
    Scale Alpha=&FloatFromCurve(FloatCurve=DataInterface() {propertyValues {Curve {keys [<at 0 1>,<at 1 0>]}}})
    Scale RGB=&MakeVectorFromLinearColorRGB(Color=&FNTimeOfDayColor(
      Night Color=&ColorFromCurve(ColorCurve=DataInterface() {propertyValues {AlphaCurve {keys [<at 0 1>,<at 1 0>]}, BlueCurve {keys [<at 0 0.4207>]}, GreenCurve {keys [<at 0 0.215>]}, RedCurve {keys [<at 0 0.0788>]}}})
      Morning Color=&ColorFromCurve(ColorCurve=DataInterface() {propertyValues {AlphaCurve {keys [<at 0 1>,<at 1 0>]}, BlueCurve {keys [<at 0 0.41>]}, GreenCurve {keys [<at 0 0.4061>]}, RedCurve {keys [<at 0 0.5689>]}}})
      Day Color=&ColorFromCurve(ColorCurve=DataInterface() {propertyValues {AlphaCurve {keys [<at 0 1>,<at 1 0>]}, BlueCurve {keys [<at 0 0.4647>]}, GreenCurve {keys [<at 0 0.3727>]}, RedCurve {keys [<at 0 0.2447>]}}})
      Evening Color=&ColorFromCurve(ColorCurve=DataInterface() {propertyValues {AlphaCurve {keys [<at 0 1>,<at 1 0>]}, BlueCurve {keys [<at 0 0.4207>]}, GreenCurve {keys [<at 0 0.2761>]}, RedCurve {keys [<at 0 0.1804>]}}})
     ))
    ScaleA=true
    ScaleRGB=true
   ) {module=ScaleColor001}
  ]
  rendererClasses [&NiagaraSpriteRendererProperties]
  renderers [{rendererClass=&NiagaraSpriteRendererProperties}]
 }
 Shockwave002 {
  emitterUpdateScript [
   &EmitterState(Life Cycle Mode=::NewEnumerator1, Loop Behavior=::NewEnumerator1, Loop Duration=1, UseLoopDelay=false)
   &SpawnBurst_Instantaneous(Use Spawn Probability=false, UseLoopCountLimit=false)
  ]
  particleSpawnScript [
   &InitializeParticle(
    Color=&FNTimeOfDayColor(
     Night Color=<rgba 0.1534 0.3719 1 0.4647>
     Morning Color=<rgba 0.5025 0.7621 1 0.4647>
     Day Color=<rgba 0.5025 0.7621 1 0.4647>
     Evening Color=<rgba 0.4023 0.7084 1 0.4647>
    )
    Lifetime=0.25
    Mesh Scale=&Multiply_VectorByFloat(Vector=<-0.4591 1 1>, Float=1.4806)
    Mesh Scale Mode=::NewEnumerator4
    Position Offset=<-1.1926 0 0>
    Sprite Size Mode=::NewEnumerator3
    Uniform Sprite Size=50
    UsePositionOffset=true
    Write Material Random=true
   )
   &AddVelocity(Velocity=<50 0 0>)
   &ApplyInitialForces()
  ]
  particleUpdateScript [
   &ParticleState()
   &ScaleMeshSize(Scale Factor=&Multiply_VectorByFloat(
     Vector=<1 1 1>
     Float=&FloatFromCurve(
      FloatCurve=DataInterface() {propertyValues {Curve {keys [
          {arriveTangent=6.0635, interpMode=RCIM_Cubic, leaveTangent=6.0635, tangentMode=RCTM_User, time=0, value=0.5774}
          {arriveTangent=1.455, interpMode=RCIM_Cubic, leaveTangent=1.455, time=0.2, value=1.45}
          {interpMode=RCIM_Cubic, tangentMode=RCTM_User, time=1, value=2.0324}
         ]}}}
      Scale Curve=1.5
     )
    ))
   &ScaleColor(
    Scale Alpha=&FloatFromCurve(FloatCurve=DataInterface() {propertyValues {Curve {keys [
         {arriveTangent=-3.1078, interpMode=RCIM_Cubic, leaveTangent=-3.1078, tangentMode=RCTM_User, time=0, value=1}
         {interpMode=RCIM_Cubic, tangentMode=RCTM_User, time=1, value=0}
        ]}}})
    ScaleA=true
    ScaleRGB=true
   ) {enabled=false}
   &SolveForcesAndVelocity(Clamp Velocity=false, Limit Acceleration=false)
   &DynamicMaterialParameters(
    Index 0 Param 1=&FloatFromCurve(FloatCurve=DataInterface() {propertyValues {Curve {keys [
         {interpMode=RCIM_Cubic, tangentMode=RCTM_User, time=0, value=0.3509}
         {arriveTangent=-1.2591, interpMode=RCIM_Cubic, leaveTangent=-1.2591, tangentMode=RCTM_User, time=1, value=-1.3}
        ]}}})
    Module.Index0Param0Write=true
    Module.Index0Param1Write=false
    Module.Index0Param2Write=false
    Module.Index0Param3Write=false
   )
  ]
  rendererClasses [&NiagaraSpriteRendererProperties, &NiagaraMeshRendererProperties]
  renderers [
   {rendererClass=&NiagaraSpriteRendererProperties}
   {rendererClass=&NiagaraMeshRendererProperties, rendererIndex=1}
  ]
 }
 SpikeStreaksParticle001 {
  emitterUpdateScript [
   &EmitterState(
    Enable Distance Culling=true
    Life Cycle Mode=::NewEnumerator1
    Loop Behavior=::NewEnumerator1
    Loop Duration=1
    Max Distance Response=::KillImmediately
    MaxDistance=3500
    Module.bUseMaxDistance=true
    Module.bUseMinDistance=false
    Scalability Mode=::NewEnumerator1
    UseLoopDelay=false
   )
   &SpawnBurst_Instantaneous(Spawn Count=6, Use Spawn Probability=false)
  ]
  particleSpawnScript [
   /Niagara/Modules/Spawn/Initialization/InitializeParticle(
    Color=<rgba 0.0253 0.2699 0.4705 5>
    Lifetime=&UniformRangedFloat(Minimum=0.1, Maximum=0.07)
    Mesh Scale=<1 1 0>
    Module.Write Color=true
    Module.Write Lifetime=true
    Module.Write Mass=true
    Module.Write Material Random=true
    Module.Write Position=true
    Module.Write Scale=true
    Module.Write SpriteRotation=false
    Module.Write SpriteSize=true
    Position=&Add_Vector(A=/Niagara/DynamicInputs/Helpers/SimulationPosition(), B=<0 0 0>)
    Sprite Size=&Multiply_Vector2D_ByFloat(
     Float=&FN_FrontEndFloatModifier(Gameplay=1, Front End=0.5)
     Vector2D=&UniformRangedVector2D(Minimum=<16.3749 25.9083>, Maximum=<20.0987 35.6847>)
    )
   ) {enabled=false}
   /Niagara/Modules/Spawn/Initialization/V2/InitializeParticle(
    Color=&FNTimeOfDayColor(
     Night Color=<rgba 0.0761 0.1548 0.3096 0.3>
     Morning Color=<rgba 0.1163 0.2171 0.3096 0.4>
     Day Color=<rgba 0.1163 0.2171 0.3096 0.4>
     Evening Color=<rgba 0.1163 0.2171 0.3096 0.4>
    )
    Lifetime=&UniformRangedFloat(Minimum=0.1, Maximum=0.14)
    Position Offset=<4.8681 0 0>
    Sprite Size=&Multiply_Vector2D_ByFloat(Float=0.8293, Vector2D=&RandomRangeVector2D(Minimum=<14.4133 16.6348>, Maximum=<5.8655 17.4526>))
    Sprite Size Mode=::NewEnumerator4
    UsePositionOffset=true
    Write Material Random=true
   ) {module=InitializeParticle001}
   &AddVelocityFromPoint(Module.Use Falloff Distance=false) {enabled=false}
   &AddVelocity(
    Velocity=&Add_Vector(
     A=&SelectVectorFromArray(
      Direct Array Index=&ReturnExecIndex()
      Vector Selection Array=DataInterface() {propertyValues {FloatData [<3 0 0>, <3 0 0.2162>, <3 0 -0.2082>]}}
     )
     B=&RandomRangeVector(Minimum=<2.0041 7.9864 11.7815>, Maximum=<-0.05 -8.8866 -11.8573>)
    )
    Velocity Speed Scale=15
   ) {module=AddVelocity003}
   &ShapeLocation(MaskSpawnGroup=false, Override Seed=false, Sphere Radius=1, UseOwnerScale=true)
  ]
  particleUpdateScript [
   &ParticleState()
   &CameraOffset(Camera Offset Amount=10)
   &SolveForcesAndVelocity(Clamp Velocity=false, Limit Acceleration=false)
   &ScaleSpriteSize(Scale Factor=&Vector2DFromFloat(Value=&FloatFromCurve(FloatCurve=DataInterface() {propertyValues {Curve {keys [<at 0 0.6>,<at 1 2>]}}}, Scale Curve=0.85)))
   &ScaleColor(
    Scale Alpha=&FloatFromCurve(FloatCurve=DataInterface() {propertyValues {Curve {keys [<at 0 1>,<at 1 0>]}}}, Scale Curve=1.5)
    Scale RGB=&VectorFromFloat(Value=0.4)
    ScaleA=true
    ScaleRGB=true
   )
  ]
  rendererClasses [&NiagaraSpriteRendererProperties]
  renderers [{rendererClass=&NiagaraSpriteRendererProperties}]
 }
 Toon_Impact_Flare {
  emitterUpdateScript [
   &EmitterState(Life Cycle Mode=::NewEnumerator1, Loop Behavior=::NewEnumerator1, Loop Duration=0.5, UseLoopDelay=false)
   &SpawnBurst_Instantaneous(Spawn Count=1, Use Spawn Probability=false)
  ]
  particleSpawnScript [
   &InitializeParticle(
    Lifetime=0.25
    Mesh Scale=&Multiply_VectorByFloat(Vector=&RandomRangeVector(Minimum=<1.45 1.45 0.7909>, Maximum=<1.3 1.3 0.9708>), Float=0.5787)
    Mesh Scale Mode=::NewEnumerator4
    Position Offset=<0 4.0706 0>
    Sprite Size=<200 175>
    Sprite Size Mode=::NewEnumerator4
    UsePositionOffset=false
    Write Material Random=true
   ) {module=InitializeParticle001}
   /SetVariables(Particles.CoronaSpriteSize=&Vector2DFromFloat(Value=90)) {isSetParameters=true}
   /SetVariables(Particles.ImpactColor:LinearColor=&MakeLinearColorFromVectorAndFloat("Vector (RGB)"=<1 1 1>, "Float (Alpha)"=&Multiply_Float(A=1, B=1))) {isSetParameters=true}
   &InitialMeshOrientation(
    Orientation Axis=<0 0 1>
    Orientation Vector=<0 0 1>
    Rotation=&RandomRangeVector(Minimum=<0 0 0>, Maximum=<1 0 0>)
    Use Orientation Vector=true
    Use Rotation Vector=true
   ) {module=InitialMeshOrientation001}
   &CameraOffset(Camera Offset Amount=10)
   &SystemLocation(MaskSpawnGroup=false, Offset=<1 0 0>)
   &AddVelocity(Velocity=<10 0 0>, Velocity Speed Scale=10) {enabled=false}
   &ApplyInitialForces()
  ]
  particleUpdateScript [
   &ParticleState()
   &DynamicMaterialParameters(
    Index 0 Param 1=&FloatFromCurve(FloatCurve=DataInterface() {propertyValues {Curve {keys [<at 0 1 RCIM_Cubic>, {arriveTangent=-1.25, interpMode=RCIM_Cubic, leaveTangent=-1.25, time=0.3, value=0.9}, <at 0.8 0 RCIM_Cubic>]}}})
    Index 0 Param 2=2
    Index 0 Param 3=1
    Index 0 Param 4=1
    Index 1 Param 1=&FloatFromCurve(FloatCurve=DataInterface() {propertyValues {Curve {keys [
         {arriveTangent=-2.2046, interpMode=RCIM_Cubic, leaveTangent=-2.2046, tangentMode=RCTM_User, time=0, value=2}
         {arriveTangent=-0.3702, interpMode=RCIM_Cubic, leaveTangent=-0.3702, tangentMode=RCTM_User, time=1, value=0.4}
        ]}}})
    Index 1 Param 2=0.4
    Index 1 Param 3=0.6
    Index 1 Param 4=1
    Index 2 Param 1=&FloatFromCurve(FloatCurve=DataInterface() {propertyValues {Curve {keys [<at 0.46 0>,<at 1 1>]}}})
    Index 2 Param 2=45
    Module.Param0WriteEnabled=true
    Module.Param0WriteEnabled1=true
    Module.Param0WriteEnabled2=true
    Module.Param1WriteEnabled=true
    Module.Param1WriteEnabled1=true
    Module.Param1WriteEnabled2=true
    Module.Param2WriteEnabled=true
    Module.Param2WriteEnabled1=true
    Module.Param2WriteEnabled2=true
    Module.Param3WriteEnabled=true
    Module.Param3WriteEnabled1=true
    Module.Param3WriteEnabled2=true
    Write Parameter Index 1=true
    Write Parameter Index 2=true
   )
   /SetVariables(Particles.ImpactColor:LinearColor=&MakeLinearColor(R:NiagaraFloat=0, G:NiagaraFloat=0, B:NiagaraFloat=0, A:NiagaraFloat=0)) {isSetParameters=true}
   &ScaleMeshSize(Scale Factor=&VectorFromFloat(Value=&FloatFromCurve(FloatCurve=DataInterface() {propertyValues {Curve {keys [<at 0 0.8071 RCIM_Cubic>,<at 1 1.202>]}}}, Scale Curve=1.5)))
   &Drag(Drag=5, Use Linear Drag=true, Use Rotational Drag=true)
   &SolveForcesAndVelocity(Clamp Velocity=false, Limit Acceleration=false)
   &ScaleColor(
    Scale Alpha=&FNTimeOfDayScalar(Night Scale=0.3, Morning Scale=0.4, Day Scale=0.4, Evening Scale=0.4)
    Scale RGB=&MakeVectorFromLinearColorRGB(Color=&FNTimeOfDayColor(
      Night Color=<rgba 0.1534 0.302 1 0.1>
      Morning Color=<rgba 0.2506 0.599 1 0.2>
      Day Color=<rgba 0.2506 0.599 1 0.2>
      Evening Color=<rgba 0.2506 0.599 1 0.2>
     ))
    ScaleA=true
    ScaleRGB=true
   )
  ]
  rendererClasses [&NiagaraMeshRendererProperties]
  renderers [{rendererClass=&NiagaraMeshRendererProperties, rendererIndex=0}]
 }
 X {
  emitterUpdateScript [
   &EmitterState(Life Cycle Mode=::NewEnumerator1, Loop Behavior=::NewEnumerator1, Loop Duration=0.4, UseLoopDelay=false)
   &SpawnBurst_Instantaneous(Spawn Count=5, Use Spawn Probability=false, UseLoopCountLimit=false)
  ]
  particleSpawnScript [
   &InitializeParticle(
    Color=<rgba 1.25 1.25 1.25 1>
    Lifetime=0.28
    Position Offset=<10 0 0>
    Sprite Size=<7.5 25>
    Sprite Size Mode=::NewEnumerator4
    Sprite UV Mode=::NewEnumerator1
    UsePositionOffset=false
    Write Material Random=true
   ) {module=InitializeParticle001}
   &SubUVAnimation(End Frame=3)
   &ShapeLocation(
    MaskSpawnGroup=false
    Offset Mode=::NewEnumerator0
    Override Seed=false
    Ring / Disc Distribution Mode:ENiagaraSphereDistributionMode=::NewEnumerator1
    Ring Radius=5
    Rotation Angle=0.125
    Rotation Mode=::NewEnumerator1
    Shape Primitive=::NewEnumerator4
    U Position=&ReturnNormalizedExecIndex()
    UseOwnerScale=true
   )
   &AddVelocity(Module.Use Falloff Distance=false, Override Seed=false, Velocity Mode=::NewEnumerator1, Velocity Speed=15) {module=AddVelocity003}
   &CameraOffset(Camera Offset Amount=10)
  ]
  particleUpdateScript [
   &ParticleState()
   &Drag(
    Drag=&FloatFromCurve(
     FloatCurve=DataInterface() {propertyValues {Curve {keys [{arriveTangent=29.6162, interpMode=RCIM_Cubic, leaveTangent=29.6162, tangentMode=RCTM_User, time=0, value=0}, <at 0.1 1 RCIM_Cubic>]}}}
     Scale Curve=5
    )
    Use Linear Drag=true
    Use Rotational Drag=true
   )
   &SolveForcesAndVelocity(Clamp Velocity=false, Limit Acceleration=false)
   &ScaleSpriteSize(Scale Factor=&Vector2DFromCurve(
     Vector2Curve=DataInterface() {propertyValues {
       XCurve {keys [<at 0 1 RCIM_Cubic>, {arriveTangent=-2.278, interpMode=RCIM_Cubic, leaveTangent=-2.278, tangentMode=RCTM_User, time=1, value=0.2473}]}
       YCurve {keys [<at 0 1.028 RCIM_Cubic>, {arriveTangent=5.991, interpMode=RCIM_Cubic, leaveTangent=5.991, tangentMode=RCTM_User, time=1, value=2.8057}]}
      }}
     ScaleCurve=&Multiply_Vector2D_ByFloat(Float=1, Vector2D=<1 2>)
    ))
   &ScaleColor(
    Scale Alpha=&FloatFromCurve(
     FloatCurve=DataInterface() {propertyValues {Curve {keys [<at 0 1 RCIM_Cubic>, {arriveTangent=-1.851, tangentMode=RCTM_User, time=1, value=0}]}}}
     Scale Curve=&FNTimeOfDayScalar(Night Scale=0.7, Morning Scale=0.8, Evening Scale=0.8)
    )
    Scale RGB=&MakeVectorFromLinearColorRGB(Color=&FNTimeOfDayColor(
      Night Color=&ColorFromCurve(ColorCurve=DataInterface() {propertyValues {
         AlphaCurve {keys []}
         BlueCurve {keys [<at 0 0.72 RCIM_Cubic>,<at 0.7087 0.7772 RCIM_Cubic>]}
         GreenCurve {keys [<at 0 0.2282 RCIM_Cubic>,<at 0.7087 0.2143 RCIM_Cubic>]}
         RedCurve {keys [<at 0 0.1092 RCIM_Cubic>,<at 0.7087 0.0378 RCIM_Cubic>]}
        }})
      Morning Color=&ColorFromCurve(ColorCurve=DataInterface() {propertyValues {
         AlphaCurve {keys []}
         BlueCurve {keys [<at 0 0.72 RCIM_Cubic>,<at 0.7087 0.7772 RCIM_Cubic>]}
         GreenCurve {keys [<at 0 0.4498 RCIM_Cubic>,<at 0.7087 0.4736 RCIM_Cubic>]}
         RedCurve {keys [<at 0 0.2704 RCIM_Cubic>,<at 0.7087 0.1658 RCIM_Cubic>]}
        }})
      Day Color=&ColorFromCurve(ColorCurve=DataInterface() {propertyValues {
         AlphaCurve {keys []}
         BlueCurve {keys [<at 0 0.72 RCIM_Cubic>,<at 0.7087 0.7772 RCIM_Cubic>]}
         GreenCurve {keys [<at 0 0.4498 RCIM_Cubic>,<at 0.7087 0.4736 RCIM_Cubic>]}
         RedCurve {keys [<at 0 0.2704 RCIM_Cubic>,<at 0.7087 0.1658 RCIM_Cubic>]}
        }})
      Evening Color=&ColorFromCurve(ColorCurve=DataInterface() {propertyValues {
         AlphaCurve {keys []}
         BlueCurve {keys [<at 0 0.72 RCIM_Cubic>,<at 0.7087 0.7772 RCIM_Cubic>]}
         GreenCurve {keys [<at 0 0.4498 RCIM_Cubic>,<at 0.7087 0.4736 RCIM_Cubic>]}
         RedCurve {keys [<at 0 0.2704 RCIM_Cubic>,<at 0.7087 0.1658 RCIM_Cubic>]}
        }})
     ))
    ScaleA=true
    ScaleRGB=true
   )
   /SetVariables(Particles.Position=&AddVectorToPosition(
     Position=<0 0 0>
     Vector=&VectorFromCurve(VectorCurve=DataInterface() {propertyValues {XCurve {keys [<at 0 0 RCIM_Cubic>,<at 1 10 RCIM_Cubic>]}, YCurve {keys [<at 0 0>,<at 1 0>]}, ZCurve {keys [<at 0 0>,<at 1 0>]}}})
    )) {enabled=false, isSetParameters=true}
  ]
  rendererClasses [&NiagaraSpriteRendererProperties]
  renderers [{rendererClass=&NiagaraSpriteRendererProperties}]
 }
}
summary {
 systemName=NS_Emote_TwilightSpot_Impact
 emitters [
  {emitterName=NE_Impact, rendererClasses [&NiagaraSpriteRendererProperties]}
  {emitterName=Toon_Impact_Flare, rendererClasses [&NiagaraMeshRendererProperties]}
  {emitterName=NE_Sparks, rendererClasses [&NiagaraSpriteRendererProperties]}
  {bEnabled=false, emitterName=NE_Impact001, rendererClasses [&NiagaraSpriteRendererProperties]}
  {emitterName=Shockwave002, rendererClasses [&NiagaraSpriteRendererProperties, &NiagaraMeshRendererProperties]}
  {emitterName=SpikeStreaksParticle001, rendererClasses [&NiagaraSpriteRendererProperties]}
  {emitterName=X, rendererClasses [&NiagaraSpriteRendererProperties]}
 ]
 userVariables [
  User.bIsFrontEnd {type=NiagaraBool}
  User.bIsFrontEndPreview {type=NiagaraBool}
  User.IsFrontEnd {type=NiagaraBool}
 ]
}
