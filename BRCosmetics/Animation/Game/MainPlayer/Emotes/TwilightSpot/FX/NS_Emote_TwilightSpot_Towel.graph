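# Niagara effect for the towel in the TwilightSpot emote, drawing a stream of short-lived, spinning mesh rings, presumably a swirl or motion blur around the towel.
# Main_Ring and BlurRing spawn meshes spread mostly along Y, offset toward camera and scaled by position length.
# Curves drive fade and a material parameter, while time of day tints both layers.
# BlurRing writes separate color and scale attributes, with front-end-adjusted alpha.
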NiagaraSystem /BRCosmetics/Animation/Game/MainPlayer/Emotes/TwilightSpot/FX/NS_Emote_TwilightSpot_Towel
dependencies {
 usedDataInterfaces [&NiagaraDataInterfaceCurve]
 usedDynamicInputs [&MakeVector, &RandomRangeFloat]
 usedModules [&SystemState]
}
emitters {
 BlurRing {
  emitterUpdateScript [
   &EmitterState(Life Cycle Mode=::NewEnumerator1, Loop Behavior=::NewEnumerator1, Loop Duration=4, UseLoopDelay=false)
   &SpawnRate(SpawnRate=9, Use Spawn Probability=false)
  ]
  particleSpawnScript [
   &InitializeParticle(
    Color Maximum=<rgba 0.4 0.47 0.5 0.6>
    Color Minimum=<rgba 0.8 0.9367 1 0.7>
    Color Mode=::NewEnumerator2
    Lifetime=0.25
    Mesh Scale=&Multiply_VectorByFloat(Vector=<1 1.7651 1>, Float=0.1216)
    Mesh Scale Mode=::NewEnumerator4
    Position Offset=&RandomRangeVector(Minimum=<-1 -12 -1>, Maximum=<1 12 1>)
    UsePositionOffset=true
    Write Material Random=true
   )
   &CameraOffset(Camera Offset Amount=15)
   &InitialMeshOrientation(
    Mesh Orientation Mode=::NewEnumerator1
    Rotation=&RandomRangeVector(Minimum=<0 0 0>, Maximum=<0 1 0>)
    Use Rotation Vector=true
   )
  ]
  particleUpdateScript [
   &ParticleState()
   &ScaleMeshSize(Scale Factor=&Multiply_VectorByFloat(
     Vector=<1 1 1>
     Float=&Lerp_Float(
      A=0.9
      B=0.5
      Alpha=&Multiply_Float(
       A=&VectorLength(Vector=&ConvertPositionToVector(Input Position=$Particles.Position, Conversion Operation:ENiagara_LWCConvertPosToVec=::NewEnumerator0))
       B=0.0802
      )
     )
    )) {module=ScaleMeshSize001}
   &UpdateMeshOrientation(Rotation Rate=2, Rotation Vector=<0 1 0>) {module=UpdateMeshOrientation002}
   &DynamicMaterialParameters(
    Index 0 Param 1=&FloatFromCurve(FloatCurve=DataInterface() {propertyValues {Curve {keys [
         {arriveTangent=2.4535, interpMode=RCIM_Cubic, leaveTangent=2.4535, tangentMode=RCTM_User, time=0, value=0}
         {arriveTangent=1, interpMode=RCIM_Cubic, leaveTangent=1, time=0.0239, value=0.3818}
         {interpMode=RCIM_Cubic, tangentMode=RCTM_User, time=0.5, value=0.5}
         {arriveTangent=-2.5865, tangentMode=RCTM_User, time=1, value=0}
        ]}}})
    Module.Index0Param0Write=true
    Module.Index0Param1Write=false
    Module.Index0Param2Write=false
    Module.Index0Param3Write=false
   )
   &ScaleColor(
    Scale Alpha=&FloatFromCurve(FloatCurve=DataInterface() {propertyValues {Curve {keys [
         {arriveTangent=6.3511, interpMode=RCIM_Cubic, leaveTangent=6.3511, tangentMode=RCTM_User, time=0, value=0}
         {interpMode=RCIM_Cubic, tangentMode=RCTM_User, time=0.5, value=1}
         {arriveTangent=-6.5462, tangentMode=RCTM_User, time=1, value=0}
        ]}}})
    Scale RGB=<0.5 0.5 0.5>
    ScaleA=true
    ScaleRGB=true
   )
   /SetVariables(Particles.New Linear Color:LinearColor=&MakeLinearColorFromVectorAndFloat(
     "Vector (RGB)"=&MakeVectorFromLinearColorRGB(Color=&FNTimeOfDayColor(
       Night Color=<rgba 0.1253 0.1914 0.5 1>
       Morning Color=<rgba 0.5 0.3383 0.3719 1>
       Day Color=<rgba 0.5 0.5 0.5 1>
       Evening Color=<rgba 0.4474 0.4547 0.4578 1>
      ))
     "Float (Alpha)"=&Multiply_Float(
      A=&Multiply_Float(
       A=&FN_FrontEndFloatModifier(Gameplay=1, Front End=1.4)
       B=&FNTimeOfDayScalar(Night Scale=0.5, Morning Scale=0.3, Evening Scale=0.3)
      )
      B=&FloatFromCurve(FloatCurve=DataInterface() {propertyValues {Curve {keys [
           {arriveTangent=2, interpMode=RCIM_Cubic, time=0.0164, value=0.8082}
           {interpMode=RCIM_Cubic, tangentMode=RCTM_User, time=0.5, value=1}
           {tangentMode=RCTM_User, time=1, value=0}
          ]}}})
     )
    )) {isSetParameters=true}
   /SetVariables(Particles.BlurMeshScale:Vector3f=&Multiply_Vector(A=$Particles.Initial.Scale, B=<0.85 0.85 0.85>)) {isSetParameters=true}
  ]
  rendererClasses [&NiagaraMeshRendererProperties]
  renderers [{rendererClass=&NiagaraMeshRendererProperties, rendererIndex=0}]
 }
 Main_Ring {
  emitterUpdateScript [
   &EmitterState(Life Cycle Mode=::NewEnumerator1, Loop Behavior=::NewEnumerator1, Loop Duration=4, UseLoopDelay=false)
   &SpawnRate(SpawnRate=20, Use Spawn Probability=false)
  ]
  particleSpawnScript [
   &PartitionParticles()
   &InitializeParticle(
    Color=&RandomRangeLinearColor(Minimum=<rgba 0.4617 0.5413 0.599 0.75>, Maximum=<rgba 0.2289 0.3396 0.5052 1>)
    Lifetime=0.25
    Mesh Scale Max=&Multiply_VectorByFloat(Vector=<1 1 1>, Float=0.18)
    Mesh Scale Min=&Multiply_VectorByFloat(Vector=<1 1 1>, Float=0.14)
    Mesh Scale Mode=::NewEnumerator2
    Position Offset=&RandomRangeVector(Minimum=<-1 -10 -1>, Maximum=<1 10 1>)
    UsePositionOffset=true
    Write Material Random=true
   )
   &CameraOffset(Camera Offset Amount=15)
   &InitialMeshOrientation(
    Mesh Orientation Mode=::NewEnumerator1
    Rotation=&RandomRangeVector(Minimum=<0 0 0>, Maximum=<0 1 0>)
    Use Rotation Vector=true
   )
  ]
  particleUpdateScript [
   &ParticleState()
   &ScaleMeshSize(Scale Factor=&Multiply_VectorByFloat(
     Vector=<1 1 1>
     Float=&Lerp_Float(
      A=1
      B=0.4
      Alpha=&Multiply_Float(
       A=&VectorLength(Vector=&ConvertPositionToVector(Input Position=$Particles.Position, Conversion Operation:ENiagara_LWCConvertPosToVec=::NewEnumerator0))
       B=0.0802
      )
     )
    )) {module=ScaleMeshSize001}
   &UpdateMeshOrientation(Rotation Rate=0.5, Rotation Vector=<0 1 0>) {module=UpdateMeshOrientation002}
   &ScaleMeshSize(Scale Factor=&VectorFromFloat(Value=&FloatFromCurve(FloatCurve=DataInterface() {propertyValues {Curve {keys [<at 0 0.7>,<at 1 1>]}}})))
   &DynamicMaterialParameters(
    Index 0 Param 1=&FloatFromCurve(FloatCurve=DataInterface() {propertyValues {Curve {keys [
         {arriveTangent=2.4535, interpMode=RCIM_Cubic, leaveTangent=2.4535, tangentMode=RCTM_User, time=0, value=0}
         {interpMode=RCIM_Cubic, tangentMode=RCTM_User, time=0.5, value=0.5}
         {arriveTangent=-2.5865, tangentMode=RCTM_User, time=1, value=0}
        ]}}})
    Module.Index0Param0Write=true
    Module.Index0Param1Write=false
    Module.Index0Param2Write=false
    Module.Index0Param3Write=false
   )
   &ScaleColor(
    Scale Alpha=&FloatFromCurve(FloatCurve=DataInterface() {propertyValues {Curve {keys [
         {arriveTangent=6.3511, interpMode=RCIM_Cubic, leaveTangent=6.3511, tangentMode=RCTM_User, time=0, value=0}
         {interpMode=RCIM_Cubic, tangentMode=RCTM_User, time=0.5, value=1}
         {tangentMode=RCTM_User, time=1, value=0}
        ]}}})
    Scale RGB=&MakeVectorFromLinearColorRGB(Color=&FNTimeOfDayColor(
      Night Color=<rgba 0.1253 0.1914 0.5 1>
      Morning Color=<rgba 0.5 0.3383 0.3719 1>
      Day Color=<rgba 0.5 0.5 0.5 1>
      Evening Color=<rgba 0.4474 0.4547 0.4578 1>
     ))
    ScaleA=true
    ScaleRGB=true
   )
  ]
  rendererClasses [&NiagaraMeshRendererProperties]
  renderers [{rendererClass=&NiagaraMeshRendererProperties, rendererIndex=0}]
 }
}
summary {
 systemName=NS_Emote_TwilightSpot_Towel
 emitters [
  {emitterName=Main_Ring, rendererClasses [&NiagaraMeshRendererProperties]}
  {emitterName=BlurRing, rendererClasses [&NiagaraMeshRendererProperties]}
 ]
 userVariables [
  User.bIsFrontEnd {type=NiagaraBool}
  User.bIsFrontEndPreview {type=NiagaraBool}
 ]
}
